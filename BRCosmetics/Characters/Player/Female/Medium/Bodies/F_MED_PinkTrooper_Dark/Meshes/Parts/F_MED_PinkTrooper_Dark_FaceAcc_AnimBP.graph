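# AnimBlueprint for the F_MED_PinkTrooper_Dark face accessory skeleton, adding secondary hair motion to this female character cosmetic.
# A chain of sprung AnimDynamics nodes simulates main hair bones, using spherical limits on head, neck, shoulders and spine.
# When bIsDBNOCarried is true, the blend bypasses that pass for the cached AimConstraints pose.
# Weighted rotation constraints then mix neighboring strands into _cst bones, followed by further AnimDynamics on mid, end and nape bones.
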AnimBlueprint /BRCosmetics/Characters/Player/Female/Medium/Bodies/F_MED_PinkTrooper_Dark/Meshes/Parts/F_MED_PinkTrooper_Dark_FaceAcc_AnimBP
target_skeleton=&/F_MED_PinkTrooper_Dark_FaceAcc_Skeleton
graphs [
 AnimGraph {nodes [
   n1 = Use cached pose 'AimConstraints_Cache'()
   n2 = Get bIsDBNOCarried()
   n3 = Anim Dynamics - Bone: dyn_hair_l_01(ComponentPose=Local To Component(LocalPose=Use cached pose 'AimConstraints_Cache'()), Alpha=1, ExternalForce="0,0,0") {
    angularDampingOverride=1
    angularSpringConstant=1500
    bAngularSpring=true
    bGravityOverrideInSimSpace=true
    bOverrideAngularDamping=true
    bUseSphericalLimits=true
    gravityScale=0.35
    simulationSpace=RootRelative
    alphaBoolBlend {blendOption=Linear}
    alphaScaleBiasClamp {scale=0.7}
    boundBone {boneName=dyn_hair_l_01}
    componentLinearAccScale {z=0.5}
    componentLinearVelScale {z=0.5}
    physicsBodyDefinitions [{
      collisionType=CustomSphere
      sphereCollisionRadius=2
      boundBone {boneName=dyn_hair_l_01}
      boxExtents=<10 20 20>
      constraintSetup {angularLimitsMax=<0 40 20>, angularLimitsMin=<0 -25 -40>, angularTarget=<0 0 0>}
      localJointOffset=<5 0 0>
     }]
    sphericalLimits [{limitRadius=8, drivingBone {boneName=head}, sphereLocalOffset=<0 0 0>}]
   }
   n4 = Anim Dynamics - Bone: dyn_hair_l_02(ComponentPose=@n3, Alpha=1, ExternalForce="0,0,0") {
    angularDampingOverride=1
    angularSpringConstant=1500
    bAngularSpring=true
    bGravityOverrideInSimSpace=true
    bOverrideAngularDamping=true
    bUseSphericalLimits=true
    gravityScale=0.35
    simulationSpace=RootRelative
    alphaBoolBlend {blendOption=Linear}
    alphaScaleBiasClamp {scale=0.7}
    boundBone {boneName=dyn_hair_l_02}
    componentLinearAccScale {z=0.5}
    componentLinearVelScale {z=0.5}
    physicsBodyDefinitions [{
      collisionType=CustomSphere
      sphereCollisionRadius=2
      boundBone {boneName=dyn_hair_l_02}
      boxExtents=<10 20 20>
      constraintSetup {angularLimitsMax=<0 40 20>, angularLimitsMin=<0 -25 -40>, angularTarget=<0 0 0>}
      localJointOffset=<5 0 0>
     }]
    sphericalLimits [{limitRadius=8, drivingBone {boneName=head}, sphereLocalOffset=<0 0 0>}]
   }
   n5 = Anim Dynamics - Bone: dyn_hair_r_01(ComponentPose=@n4, Alpha=1, ExternalForce="0,0,0") {
    angularDampingOverride=1
    angularSpringConstant=1500
    bAngularSpring=true
    bGravityOverrideInSimSpace=true
    bOverrideAngularDamping=true
    bUseSphericalLimits=true
    gravityScale=0.35
    simulationSpace=RootRelative
    alphaBoolBlend {blendOption=Linear}
    alphaScaleBiasClamp {scale=0.7}
    boundBone {boneName=dyn_hair_r_01}
    componentLinearAccScale {z=0.5}
    componentLinearVelScale {z=0.5}
    physicsBodyDefinitions [{
      collisionType=CustomSphere
      sphereCollisionRadius=2
      boundBone {boneName=dyn_hair_r_01}
      boxExtents=<10 20 20>
      constraintSetup {angularLimitsMax=<0 40 20>, angularLimitsMin=<0 -25 -40>, angularTarget=<0 0 0>}
      localJointOffset=<5 0 0>
     }]
    sphericalLimits [{limitRadius=7, drivingBone {boneName=head}, sphereLocalOffset=<0 0 0>}]
   }
   n6 = Anim Dynamics - Bone: dyn_hair_r_02(ComponentPose=@n5, Alpha=1, ExternalForce="0,0,0") {
    angularDampingOverride=1
    angularSpringConstant=1500
    bAngularSpring=true
    bGravityOverrideInSimSpace=true
    bOverrideAngularDamping=true
    bUseSphericalLimits=true
    gravityScale=0.35
    simulationSpace=RootRelative
    alphaBoolBlend {blendOption=Linear}
    alphaScaleBiasClamp {scale=0.7}
    boundBone {boneName=dyn_hair_r_02}
    componentLinearAccScale {z=0.5}
    componentLinearVelScale {z=0.5}
    physicsBodyDefinitions [{
      collisionType=CustomSphere
      sphereCollisionRadius=2
      boundBone {boneName=dyn_hair_r_02}
      boxExtents=<10 20 20>
      constraintSetup {angularLimitsMax=<0 40 20>, angularLimitsMin=<0 -25 -40>, angularTarget=<0 0 0>}
      localJointOffset=<5 0 0>
     }]
    sphericalLimits [{limitRadius=7, drivingBone {boneName=head}, sphereLocalOffset=<0 0 0>}]
   }
   n7 = Anim Dynamics - Bone: dyn_hair_fls(ComponentPose=@n6, Alpha=1, ExternalForce="0,0,0") {
    angularDampingOverride=1
    angularSpringConstant=1500
    bAngularSpring=true
    bGravityOverrideInSimSpace=true
    bOverrideAngularDamping=true
    bUseSphericalLimits=true
    gravityScale=0.35
    simulationSpace=RootRelative
    alphaBoolBlend {blendOption=Linear}
    alphaScaleBiasClamp {scale=0.7}
    boundBone {boneName=dyn_hair_fls}
    componentLinearAccScale {z=0.46}
    componentLinearVelScale {z=0.46}
    physicsBodyDefinitions [{
      collisionType=CustomSphere
      sphereCollisionRadius=1
      boundBone {boneName=dyn_hair_fls}
      boxExtents=<10 30 30>
      constraintSetup {angularLimitsMax=<0 40 20>, angularLimitsMin=<0 -25 -30>, angularTarget=<1 0 0>}
      localJointOffset=<5 0 0>
     }]
    sphericalLimits [
     {limitRadius=7, drivingBone {boneName=neck_02}, sphereLocalOffset=<0 0 0>}
     {limitRadius=6, drivingBone {boneName=head}, sphereLocalOffset=<0 0 0>}
     {limitRadius=10, drivingBone {boneName=clavicle_l}, sphereLocalOffset=<0 0 0>}
    ]
   }
   n8 = Anim Dynamics - Bone: dyn_hair_bls(ComponentPose=@n7, Alpha=1, ExternalForce="0,0,0") {
    angularDampingOverride=1
    angularSpringConstant=1500
    bAngularSpring=true
    bGravityOverrideInSimSpace=true
    bOverrideAngularDamping=true
    bUseSphericalLimits=true
    gravityScale=0.35
    simulationSpace=RootRelative
    alphaBoolBlend {blendOption=Linear}
    alphaScaleBiasClamp {scale=0.7}
    boundBone {boneName=dyn_hair_bls}
    componentLinearAccScale {z=0.46}
    componentLinearVelScale {z=0.46}
    physicsBodyDefinitions [{
      collisionType=CustomSphere
      sphereCollisionRadius=2
      boundBone {boneName=dyn_hair_bls}
      boxExtents=<10 30 30>
      constraintSetup {angularLimitsMax=<0 40 20>, angularLimitsMin=<0 -25 -45>, angularTarget=<1 0 0>}
      localJointOffset=<5 0 0>
     }]
    sphericalLimits [
     {limitRadius=10, drivingBone {boneName=clavicle_l}, sphereLocalOffset=<10 0 -3>}
     {limitRadius=10, drivingBone {boneName=clavicle_r}, sphereLocalOffset=<-10 0 3>}
     {limitRadius=10, drivingBone {boneName=spine_05}, sphereLocalOffset=<0 0 0>}
     {limitRadius=7.5, drivingBone {boneName=head}, sphereLocalOffset=<0 0 0>}
    ]
   }
   n9 = Anim Dynamics - Bone: dyn_hair_frs(ComponentPose=@n8, Alpha=1, ExternalForce="0,0,0") {
    angularDampingOverride=1
    angularSpringConstant=1500
    bAngularSpring=true
    bGravityOverrideInSimSpace=true
    bOverrideAngularDamping=true
    bUsePlanarLimit=false
    bUseSphericalLimits=true
    gravityScale=0.35
    simulationSpace=RootRelative
    alphaBoolBlend {blendOption=Linear}
    alphaScaleBiasClamp {scale=0.7}
    boundBone {boneName=dyn_hair_frs}
    componentLinearAccScale {z=0.46}
    componentLinearVelScale {z=0.46}
    physicsBodyDefinitions [{
      collisionType=CustomSphere
      sphereCollisionRadius=1
      boundBone {boneName=dyn_hair_frs}
      boxExtents=<10 30 30>
      constraintSetup {angularLimitsMax=<0 40 20>, angularLimitsMin=<0 -25 -80>, angularTarget=<1 0 0>}
      localJointOffset=<5 0 0>
     }]
    sphericalLimits [
     {limitRadius=12, drivingBone {boneName=clavicle_r}, sphereLocalOffset=<-6.3102 3.2252 7.3659>}
     {limitRadius=6, drivingBone {boneName=head}, sphereLocalOffset=<0 0 0>}
     {limitRadius=7, drivingBone {boneName=neck_01}, sphereLocalOffset=<0 0 0>}
    ]
   }
   n10 = Anim Dynamics - Bone: dyn_hair_r_03(ComponentPose=@n9, Alpha=1, ExternalForce="0,0,0") {
    angularDampingOverride=1
    angularSpringConstant=1500
    bAngularSpring=true
    bGravityOverrideInSimSpace=true
    bOverrideAngularDamping=true
    bUseSphericalLimits=true
    gravityScale=0.35
    simulationSpace=RootRelative
    alphaBoolBlend {blendOption=Linear}
    alphaScaleBiasClamp {scale=0.7}
    boundBone {boneName=dyn_hair_r_03}
    componentLinearAccScale {z=0.46}
    componentLinearVelScale {z=0.46}
    physicsBodyDefinitions [{
      collisionType=CustomSphere
      sphereCollisionRadius=2
      boundBone {boneName=dyn_hair_r_03}
      boxExtents=<10 30 30>
      constraintSetup {angularLimitsMax=<0 40 10>, angularLimitsMin=<0 -25 -10>, angularTarget=<1 0 0>}
      localJointOffset=<10 0 0>
     }]
    sphericalLimits [
     {limitRadius=14, drivingBone {boneName=clavicle_r}, sphereLocalOffset=<-4.9844 1.1218 5.2787>}
     {limitRadius=10, drivingBone {boneName=spine_05}, sphereLocalOffset=<0 0 0>}
     {limitRadius=6, drivingBone {boneName=head}, sphereLocalOffset=<0 0 0>}
     {limitRadius=9, drivingBone {boneName=neck_01}, sphereLocalOffset=<0 0 0>}
     {limitRadius=8, drivingBone {boneName=upperarm_r}, sphereLocalOffset=<-0.0644 -0.0693 0.0041>}
    ]
   }
   n11 = Anim Dynamics - Bone: dyn_hair_bk_01(ComponentPose=@n10, Alpha=1, ExternalForce="0,0,0") {
    angularDampingOverride=1
    angularSpringConstant=1500
    bAngularSpring=true
    bGravityOverrideInSimSpace=true
    bOverrideAngularDamping=true
    bUseSphericalLimits=true
    gravityScale=0.35
    simulationSpace=RootRelative
    alphaBoolBlend {blendOption=Linear}
    alphaScaleBiasClamp {scale=0.7}
    boundBone {boneName=dyn_hair_bk_01}
    componentLinearAccScale {z=0.46}
    componentLinearVelScale {z=0.46}
    physicsBodyDefinitions [{
      collisionType=CustomSphere
      sphereCollisionRadius=3
      boundBone {boneName=dyn_hair_bk_01}
      boxExtents=<20 50 50>
      constraintSetup {angularLimitsMax=<0 40 0>, angularLimitsMin=<0 -25 -40>, angularTarget=<1 0 0>}
      localJointOffset=<10 0 0>
     }]
    sphericalLimits [
     {limitRadius=10, drivingBone {boneName=head}, sphereLocalOffset=<0 6 0>}
     {limitRadius=13, drivingBone {boneName=spine_05}, sphereLocalOffset=<0 0 0>}
    ]
   }
   n12 = Anim Dynamics - Bone: dyn_hair_bk_02(ComponentPose=@n11, Alpha=1, ExternalForce="0,0,0") {
    angularDampingOverride=1
    angularSpringConstant=1500
    bAngularSpring=true
    bGravityOverrideInSimSpace=true
    bOverrideAngularDamping=true
    bUseSphericalLimits=true
    gravityScale=0.35
    simulationSpace=RootRelative
    alphaBoolBlend {blendOption=Linear}
    alphaScaleBiasClamp {scale=0.7}
    boundBone {boneName=dyn_hair_bk_02}
    componentLinearAccScale {z=0.46}
    componentLinearVelScale {z=0.46}
    physicsBodyDefinitions [{
      collisionType=CustomSphere
      sphereCollisionRadius=3
      boundBone {boneName=dyn_hair_bk_02}
      boxExtents=<20 50 50>
      constraintSetup {angularLimitsMax=<0 40 0>, angularLimitsMin=<0 -25 -40>, angularTarget=<1 0 0>}
      localJointOffset=<10 0 0>
     }]
    sphericalLimits [
     {limitRadius=10, drivingBone {boneName=head}, sphereLocalOffset=<0 6 0>}
     {limitRadius=13, drivingBone {boneName=spine_05}, sphereLocalOffset=<2.2989 0 -0.0296>}
    ]
   }
   n13 = Component To Local(ComponentPose=@n12)
   n14 = "Constraint - dyn_hair_bk_01_cst (RRR)"!(
    ComponentPose=Local To Component(LocalPose=Blend Poses by bool(BlendPose_0=@n1, BlendPose_1=@n13, bActiveValue=@n2) {blendType=Linear})
    Alpha=1
    ConstraintWeights_0=0.3
    ConstraintWeights_1:float=0.4
    ConstraintWeights_2:float=0.3
   ) -> (Pose) {
    alphaBoolBlend {blendOption=Linear}
    boneToModify {boneName=dyn_hair_bk_01_cst}
    constraintSetup [
     {transformType=Rotation, targetBone {boneName=dyn_hair_bk_02}}
     {transformType=Rotation, targetBone {boneName=dyn_hair_bk_01}}
     {transformType=Rotation, targetBone {boneName=dyn_hair_bls}}
    ]
    constraintWeights [1,1,1]
   }
   n15 = "Constraint - dyn_hair_bk_02_cst (RRR)"!(ComponentPose=@n14, Alpha=1, ConstraintWeights_0=0.4, ConstraintWeights_1:float=0.3, ConstraintWeights_2:float=0.3) -> (Pose) {
    alphaBoolBlend {blendOption=Linear}
    boneToModify {boneName=dyn_hair_bk_02_cst}
    constraintSetup [
     {transformType=Rotation, targetBone {boneName=dyn_hair_bk_02}}
     {transformType=Rotation, targetBone {boneName=dyn_hair_bk_01}}
     {transformType=Rotation, targetBone {boneName=dyn_hair_r_03}}
    ]
    constraintWeights [1,1,1]
   }
   n16 = "Constraint - dyn_hair_r_03_cst (RRR)"!(ComponentPose=@n15, Alpha=1, ConstraintWeights_0=0, ConstraintWeights_1:float=0.3, ConstraintWeights_2:float=0.7) -> (Pose) {
    alphaBoolBlend {blendOption=Linear}
    boneToModify {boneName=dyn_hair_r_03_cst}
    constraintSetup [
     {transformType=Rotation, targetBone {boneName=dyn_hair_bk_02}}
     {transformType=Rotation, targetBone {boneName=dyn_hair_frs}}
     {transformType=Rotation, targetBone {boneName=dyn_hair_r_03}}
    ]
    constraintWeights [1,1,1]
   }
   n17 = "Constraint - dyn_hair_frs_cst (RRR)"!(ComponentPose=@n16, Alpha=1, ConstraintWeights_0=0.3, ConstraintWeights_1:float=0.2, ConstraintWeights_2:float=0.5) -> (Pose) {
    alphaBoolBlend {blendOption=Linear}
    boneToModify {boneName=dyn_hair_frs_cst}
    constraintSetup [
     {transformType=Rotation, targetBone {boneName=dyn_hair_r_02}}
     {transformType=Rotation, targetBone {boneName=dyn_hair_r_03}}
     {transformType=Rotation, targetBone {boneName=dyn_hair_frs}}
    ]
    constraintWeights [1,1,1]
   }
   n18 = "Constraint - dyn_hair_fls_cst (RRR)"!(ComponentPose=@n17, Alpha=1, ConstraintWeights_0=0.5, ConstraintWeights_1:float=0.5, ConstraintWeights_2:float=0) -> (Pose) {
    alphaBoolBlend {blendOption=Linear}
    boneToModify {boneName=dyn_hair_fls_cst}
    constraintSetup [
     {transformType=Rotation, targetBone {boneName=dyn_hair_fls}}
     {transformType=Rotation, targetBone {boneName=dyn_hair_l_02}}
     {transformType=Rotation, targetBone {boneName=dyn_hair_bls}}
    ]
    constraintWeights [1,1,1]
   }
   n19 = "Constraint - dyn_hair_bls_cst (RRR)"!(ComponentPose=@n18, Alpha=1, ConstraintWeights_0=0.5, ConstraintWeights_1:float=0, ConstraintWeights_2:float=0.5) -> (Pose) {
    alphaBoolBlend {blendOption=Linear}
    boneToModify {boneName=dyn_hair_bls_cst}
    constraintSetup [
     {transformType=Rotation, targetBone {boneName=dyn_hair_bls}}
     {transformType=Rotation, targetBone {boneName=dyn_hair_fls}}
     {transformType=Rotation, targetBone {boneName=dyn_hair_bk_01}}
    ]
    constraintWeights [1,1,1]
   }
   n20 = "Constraint - dyn_hair_r_02_cst (RRR)"!(ComponentPose=@n19, Alpha=1, ConstraintWeights_0=0.3, ConstraintWeights_1:float=0.3, ConstraintWeights_2:float=0.4) -> (Pose) {
    alphaBoolBlend {blendOption=Linear}
    boneToModify {boneName=dyn_hair_r_02_cst}
    constraintSetup [
     {transformType=Rotation, targetBone {boneName=dyn_hair_r_01}}
     {transformType=Rotation, targetBone {boneName=dyn_hair_frs}}
     {transformType=Rotation, targetBone {boneName=dyn_hair_r_02}}
    ]
    constraintWeights [1,1,1]
   }
   n21 = "Constraint - dyn_hair_l_02_cst (RRR)"!(ComponentPose=@n20, Alpha=1, ConstraintWeights_0=0, ConstraintWeights_1:float=0.3, ConstraintWeights_2:float=0.7) -> (Pose) {
    alphaBoolBlend {blendOption=Linear}
    boneToModify {boneName=dyn_hair_l_02_cst}
    constraintSetup [
     {transformType=Rotation, targetBone {boneName=dyn_hair_bls}}
     {transformType=Rotation, targetBone {boneName=dyn_hair_fls}}
     {transformType=Rotation, targetBone {boneName=dyn_hair_l_02}}
    ]
    constraintWeights [1,1,1]
   }
   n22 = "Constraint - dyn_hair_frs_cst (P)"(ComponentPose=@n21, Alpha=1, ConstraintWeights_0=1) {
    alphaBoolBlend {blendOption=Linear}
    boneToModify {boneName=dyn_hair_frs_cst}
    constraintSetup [{targetBone {boneName=head}}]
    constraintWeights [1]
   }
   n23 = "Constraint - dyn_hair_r_03_cst (RRR)"(ComponentPose=@n22, Alpha=1, ConstraintWeights_0=0.7, ConstraintWeights_1=0, ConstraintWeights_2=0.3) {
    alphaBoolBlend {blendOption=Linear}
    boneToModify {boneName=dyn_hair_r_03_cst}
    constraintSetup [
     {transformType=Rotation, targetBone {boneName=dyn_hair_bk_02}}
     {transformType=Rotation, targetBone {boneName=dyn_hair_frs}}
     {transformType=Rotation, targetBone {boneName=dyn_hair_r_03}}
    ]
    constraintWeights [1,1,1]
   }
   n24 = Anim Dynamics - Bone: dyn_hair_r_02_end(ComponentPose=@n23, Alpha=1, ExternalForce="0,0,0") {
    angularDampingOverride=1
    angularSpringConstant=1500
    bAngularSpring=true
    bGravityOverrideInSimSpace=true
    bOverrideAngularDamping=true
    bUseSphericalLimits=true
    gravityScale=0.35
    simulationSpace=RootRelative
    alphaBoolBlend {blendOption=Linear}
    alphaScaleBiasClamp {scale=0.7}
    boundBone {boneName=dyn_hair_r_02_end}
    componentLinearAccScale {z=0.46}
    componentLinearVelScale {z=0.46}
    physicsBodyDefinitions [{
      collisionType=CustomSphere
      sphereCollisionRadius=2
      boundBone {boneName=dyn_hair_r_02_end}
      boxExtents=<10 30 30>
      constraintSetup {angularLimitsMax=<0 10 0>, angularLimitsMin=<0 -10 -20>, angularTarget=<1 0 0>}
      localJointOffset=<10 0 0>
     }]
    sphericalLimits [{limitRadius=7, drivingBone {boneName=head}, sphereLocalOffset=<0 0 0>}]
   }
   n25 = Anim Dynamics - Bone: dyn_hair_frs_05_mid(ComponentPose=@n24, Alpha=1, ExternalForce="0,0,0") {
    angularDampingOverride=1
    angularSpringConstant=1500
    bAngularSpring=true
    bGravityOverrideInSimSpace=true
    bOverrideAngularDamping=true
    bUseSphericalLimits=true
    gravityScale=0.35
    simulationSpace=RootRelative
    alphaBoolBlend {blendOption=Linear}
    alphaScaleBiasClamp {scale=0.5}
    boundBone {boneName=dyn_hair_frs_05_mid}
    componentLinearAccScale {z=0.5}
    componentLinearVelScale {z=0.5}
    physicsBodyDefinitions [{
      collisionType=CustomSphere
      sphereCollisionRadius=2
      boundBone {boneName=dyn_hair_frs_05_mid}
      boxExtents=<10 20 20>
      constraintSetup {angularLimitsMax=<0 80 20>, angularLimitsMin=<0 -80 -80>, angularTarget=<0 0 0>}
      localJointOffset=<7 0 0>
     }]
    sphericalLimits [
     {limitRadius=10, drivingBone {boneName=neck_02}, sphereLocalOffset=<-10 0 0>}
     {limitRadius=10, drivingBone {boneName=clavicle_l}, sphereLocalOffset=<10 0 -3>}
     {limitRadius=9, drivingBone {boneName=clavicle_r}, sphereLocalOffset=<-10.4715 1.7833 4.3765>}
    ]
   }
   n26 = Anim Dynamics - Bone: dyn_hair_frs_05_end(ComponentPose=@n25, Alpha=1, ExternalForce="0,0,0") {
    angularDampingOverride=1
    angularSpringConstant=1500
    bAngularSpring=true
    bGravityOverrideInSimSpace=true
    bOverrideAngularDamping=true
    bUseSphericalLimits=true
    gravityScale=0.35
    simulationSpace=RootRelative
    alphaBoolBlend {blendOption=Linear}
    alphaScaleBiasClamp {scale=0.5}
    boundBone {boneName=dyn_hair_frs_05_end}
    componentLinearAccScale {z=0.5}
    componentLinearVelScale {z=0.5}
    physicsBodyDefinitions [{
      collisionType=CustomSphere
      sphereCollisionRadius=2
      boundBone {boneName=dyn_hair_frs_05_end}
      boxExtents=<10 20 20>
      constraintSetup {angularLimitsMax=<0 80 20>, angularLimitsMin=<0 -80 -80>, angularTarget=<0 0 0>}
      localJointOffset=<5 0 0>
     }]
    sphericalLimits [
     {limitRadius=7, drivingBone {boneName=neck_02}, sphereLocalOffset=<-10 0 0>}
     {limitRadius=7, drivingBone {boneName=clavicle_l}, sphereLocalOffset=<5 0 -3>}
     {limitRadius=7, drivingBone {boneName=clavicle_r}, sphereLocalOffset=<-10 0 3>}
    ]
   }
   n27 = Anim Dynamics - Bone: dyn_hair_frs_02_mid(ComponentPose=@n26, Alpha=1, ExternalForce="0,0,0") {
    angularDampingOverride=1
    angularSpringConstant=1500
    bAngularSpring=true
    bGravityOverrideInSimSpace=true
    bOverrideAngularDamping=true
    bUseSphericalLimits=true
    gravityScale=0.35
    simulationSpace=RootRelative
    alphaBoolBlend {blendOption=Linear}
    alphaScaleBiasClamp {scale=0.5}
    boundBone {boneName=dyn_hair_frs_02_mid}
    componentLinearAccScale {z=0.5}
    componentLinearVelScale {z=0.5}
    physicsBodyDefinitions [{
      collisionType=CustomSphere
      sphereCollisionRadius=2
      boundBone {boneName=dyn_hair_frs_02_mid}
      boxExtents=<20 30 30>
      constraintSetup {angularLimitsMax=<0 0 20>, angularLimitsMin=<0 0 -80>, angularTarget=<0 0 0>}
      localJointOffset=<2 0 0>
     }]
    sphericalLimits [
     {limitRadius=10, drivingBone {boneName=neck_02}, sphereLocalOffset=<-10 0 0>}
     {limitRadius=10, drivingBone {boneName=None}, sphereLocalOffset=<0 0 0>}
    ]
   }
   n28 = Anim Dynamics - Bone: dyn_hair_frs_03_mid(ComponentPose=@n27, Alpha=1, ExternalForce="0,0,0") {
    angularDampingOverride=1
    angularSpringConstant=1500
    bAngularSpring=true
    bGravityOverrideInSimSpace=true
    bOverrideAngularDamping=true
    bUseSphericalLimits=true
    gravityScale=0.35
    simulationSpace=RootRelative
    alphaBoolBlend {blendOption=Linear}
    alphaScaleBiasClamp {scale=0.5}
    boundBone {boneName=dyn_hair_frs_03_mid}
    componentLinearAccScale {z=0.5}
    componentLinearVelScale {z=0.5}
    physicsBodyDefinitions [{
      collisionType=CustomSphere
      sphereCollisionRadius=2
      boundBone {boneName=dyn_hair_frs_03_mid}
      boxExtents=<10 20 20>
      constraintSetup {angularLimitsMax=<0 10 20>, angularLimitsMin=<0 -10 -80>, angularTarget=<0 0 0>}
      localJointOffset=<10 0 0>
     }]
    sphericalLimits [
     {limitRadius=10, drivingBone {boneName=clavicle_l}, sphereLocalOffset=<10 0 -3>}
     {limitRadius=10, drivingBone {boneName=clavicle_r}, sphereLocalOffset=<-6.4071 -0.9625 6.0096>}
     {limitRadius=11, drivingBone {boneName=spine_05}, sphereLocalOffset=<0 0 0>}
     {limitRadius=8, drivingBone {boneName=neck_02}, sphereLocalOffset=<0 0 0>}
    ]
   }
   n29 = Anim Dynamics - Bone: dyn_hair_frs_03_end(ComponentPose=@n28, Alpha=1, ExternalForce="0,0,0") {
    angularDampingOverride=1
    angularSpringConstant=1500
    bAngularSpring=true
    bGravityOverrideInSimSpace=true
    bOverrideAngularDamping=true
    bUseSphericalLimits=true
    gravityScale=0.35
    simulationSpace=RootRelative
    alphaBoolBlend {blendOption=Linear}
    alphaScaleBiasClamp {scale=0.5}
    boundBone {boneName=dyn_hair_frs_03_end}
    componentLinearAccScale {z=0.5}
    componentLinearVelScale {z=0.5}
    physicsBodyDefinitions [{
      collisionType=CustomSphere
      sphereCollisionRadius=2
      boundBone {boneName=dyn_hair_frs_03_end}
      boxExtents=<10 20 20>
      constraintSetup {angularLimitsMax=<0 10 10>, angularLimitsMin=<0 -10 -10>, angularTarget=<0 0 0>}
      localJointOffset=<5 0 0>
     }]
    sphericalLimits [
     {limitRadius=10, drivingBone {boneName=neck_02}, sphereLocalOffset=<-10 0 0>}
     {limitRadius=10, drivingBone {boneName=clavicle_l}, sphereLocalOffset=<10 0 -3>}
     {limitRadius=8.5, drivingBone {boneName=clavicle_r}, sphereLocalOffset=<-10 0 3>}
     {limitRadius=13, drivingBone {boneName=spine_04}, sphereLocalOffset=<0 0 0>}
    ]
   }
   n30 = Anim Dynamics - Bone: dyn_hair_frs_06_end(ComponentPose=@n29, Alpha=1, ExternalForce="0,0,0") {
    angularDampingOverride=1
    angularSpringConstant=1500
    bAngularSpring=true
    bGravityOverrideInSimSpace=true
    bOverrideAngularDamping=true
    bUseSphericalLimits=true
    gravityScale=0.35
    simulationSpace=RootRelative
    alphaBoolBlend {blendOption=Linear}
    alphaScaleBiasClamp {scale=0.5}
    boundBone {boneName=dyn_hair_frs_06_end}
    componentLinearAccScale {z=0.5}
    componentLinearVelScale {z=0.5}
    physicsBodyDefinitions [{
      collisionType=CustomSphere
      sphereCollisionRadius=2
      boundBone {boneName=dyn_hair_frs_06_end}
      boxExtents=<10 20 20>
      constraintSetup {angularLimitsMax=<0 20 20>, angularLimitsMin=<0 -20 -80>, angularTarget=<0 0 0>}
      localJointOffset=<0 0 -10>
     }]
    sphericalLimits [{limitRadius=7, drivingBone {boneName=head}, sphereLocalOffset=<0 0 0>}]
   }
   n31 = Anim Dynamics - Bone: dyn_hair_frs_07_end(ComponentPose=@n30, Alpha=1, ExternalForce="0,0,0") {
    angularDampingOverride=1
    angularSpringConstant=1500
    bAngularSpring=true
    bGravityOverrideInSimSpace=true
    bOverrideAngularDamping=true
    bUseSphericalLimits=true
    gravityScale=0.35
    simulationSpace=RootRelative
    alphaBoolBlend {blendOption=Linear}
    alphaScaleBiasClamp {scale=0.5}
    boundBone {boneName=dyn_hair_frs_07_end}
    componentLinearAccScale {z=0.5}
    componentLinearVelScale {z=0.5}
    physicsBodyDefinitions [{
      collisionType=CustomSphere
      sphereCollisionRadius=2
      boundBone {boneName=dyn_hair_frs_07_end}
      boxExtents=<10 20 20>
      constraintSetup {angularLimitsMax=<0 20 20>, angularLimitsMin=<0 -20 -80>, angularTarget=<0 0 0>}
      localJointOffset=<5 0 0>
     }]
    sphericalLimits [{limitRadius=7, drivingBone {boneName=head}, sphereLocalOffset=<0 0 0>}]
   }
   n32 = Anim Dynamics - Bone: dyn_hair_fls_03_mid(ComponentPose=@n31, Alpha=1, ExternalForce="0,0,0") {
    angularDampingOverride=1
    angularSpringConstant=1500
    bAngularSpring=true
    bGravityOverrideInSimSpace=true
    bOverrideAngularDamping=true
    bUseSphericalLimits=true
    gravityScale=0.35
    simulationSpace=RootRelative
    alphaBoolBlend {blendOption=Linear}
    alphaScaleBiasClamp {scale=0.5}
    boundBone {boneName=dyn_hair_fls_03_mid}
    componentLinearAccScale {z=0.5}
    componentLinearVelScale {z=0.5}
    physicsBodyDefinitions [{
      collisionType=CustomSphere
      sphereCollisionRadius=2
      boundBone {boneName=dyn_hair_fls_03_mid}
      boxExtents=<10 20 20>
      constraintSetup {angularLimitsMax=<0 20 20>, angularLimitsMin=<0 -20 -80>, angularTarget=<0 0 0>}
      localJointOffset=<10 0 0>
     }]
    sphericalLimits [
     {limitRadius=7, drivingBone {boneName=neck_02}, sphereLocalOffset=<-10 0 0>}
     {limitRadius=10, drivingBone {boneName=clavicle_l}, sphereLocalOffset=<10 0 -3>}
     {limitRadius=10, drivingBone {boneName=clavicle_r}, sphereLocalOffset=<-10 0 3>}
     {limitRadius=8, drivingBone {boneName=spine_05}, sphereLocalOffset=<1.6244 1.4317 -5.2154>}
    ]
   }
   n33 = Anim Dynamics - Bone: dyn_hair_fls_01_end(ComponentPose=@n32, Alpha=1, ExternalForce="0,0,0") {
    angularDampingOverride=1
    angularSpringConstant=1500
    bAngularSpring=true
    bGravityOverrideInSimSpace=true
    bOverrideAngularDamping=true
    bUseSphericalLimits=true
    gravityScale=0.35
    simulationSpace=RootRelative
    alphaBoolBlend {blendOption=Linear}
    alphaScaleBiasClamp {scale=0.5}
    boundBone {boneName=dyn_hair_fls_01_end}
    componentLinearAccScale {z=0.5}
    componentLinearVelScale {z=0.5}
    physicsBodyDefinitions [{
      collisionType=CustomSphere
      sphereCollisionRadius=2
      boundBone {boneName=dyn_hair_fls_01_end}
      boxExtents=<10 20 20>
      constraintSetup {angularLimitsMax=<0 20 20>, angularLimitsMin=<0 -20 -80>, angularTarget=<0 0 0>}
      localJointOffset=<0 0 -5>
     }]
    sphericalLimits [
     {limitRadius=7, drivingBone {boneName=neck_02}, sphereLocalOffset=<-10 0 0>}
     {limitRadius=8, drivingBone {boneName=clavicle_l}, sphereLocalOffset=<10 0 -3>}
     {limitRadius=8, drivingBone {boneName=clavicle_r}, sphereLocalOffset=<-10 0 3>}
    ]
   }
   n34 = Anim Dynamics - Bone: dyn_hair_fls_02_end(ComponentPose=@n33, Alpha=1, ExternalForce="0,0,0") {
    angularDampingOverride=1
    angularSpringConstant=1500
    bAngularSpring=true
    bGravityOverrideInSimSpace=true
    bOverrideAngularDamping=true
    bUseSphericalLimits=true
    gravityScale=0.35
    simulationSpace=RootRelative
    alphaBoolBlend {blendOption=Linear}
    alphaScaleBiasClamp {scale=0.5}
    boundBone {boneName=dyn_hair_fls_02_end}
    componentLinearAccScale {z=0.5}
    componentLinearVelScale {z=0.5}
    physicsBodyDefinitions [{
      collisionType=CustomSphere
      sphereCollisionRadius=2
      boundBone {boneName=dyn_hair_fls_02_end}
      boxExtents=<10 20 20>
      constraintSetup {angularLimitsMax=<0 20 20>, angularLimitsMin=<0 -20 -80>, angularTarget=<0 0 0>}
      localJointOffset=<2 0 0>
     }]
    sphericalLimits [
     {limitRadius=7, drivingBone {boneName=neck_02}, sphereLocalOffset=<-10 0 0>}
     {limitRadius=10, drivingBone {boneName=clavicle_l}, sphereLocalOffset=<10 0 -3>}
     {limitRadius=10, drivingBone {boneName=clavicle_r}, sphereLocalOffset=<-10 0 3>}
    ]
   }
   n35 = Anim Dynamics - Bone: dyn_hair_fls_03_end(ComponentPose=@n34, Alpha=1, ExternalForce="0,0,0") {
    angularDampingOverride=1
    angularSpringConstant=1500
    bAngularSpring=true
    bGravityOverrideInSimSpace=true
    bOverrideAngularDamping=true
    bUseSphericalLimits=true
    gravityScale=0.35
    simulationSpace=RootRelative
    alphaBoolBlend {blendOption=Linear}
    alphaScaleBiasClamp {scale=0.5}
    boundBone {boneName=dyn_hair_fls_03_end}
    componentLinearAccScale {z=0.5}
    componentLinearVelScale {z=0.5}
    physicsBodyDefinitions [{
      collisionType=CustomSphere
      sphereCollisionRadius=2
      boundBone {boneName=dyn_hair_fls_03_end}
      boxExtents=<10 20 20>
      constraintSetup {angularLimitsMax=<0 20 20>, angularLimitsMin=<0 -20 -80>, angularTarget=<0 0 0>}
      localJointOffset=<5 0 0>
     }]
    sphericalLimits [
     {limitRadius=7, drivingBone {boneName=neck_02}, sphereLocalOffset=<-10 0 0>}
     {limitRadius=10, drivingBone {boneName=clavicle_l}, sphereLocalOffset=<10 0 -3>}
     {limitRadius=10, drivingBone {boneName=clavicle_r}, sphereLocalOffset=<-10 0 3>}
    ]
   }
   n36 = Anim Dynamics - Bone: dyn_hair_fls_04_end(ComponentPose=@n35, Alpha=1, ExternalForce="0,0,0") {
    angularDampingOverride=1
    angularSpringConstant=1500
    bAngularSpring=true
    bGravityOverrideInSimSpace=true
    bOverrideAngularDamping=true
    bUseSphericalLimits=true
    gravityScale=0.35
    simulationSpace=RootRelative
    alphaBoolBlend {blendOption=Linear}
    alphaScaleBiasClamp {scale=0.5}
    boundBone {boneName=dyn_hair_fls_04_end}
    componentLinearAccScale {z=0.5}
    componentLinearVelScale {z=0.5}
    physicsBodyDefinitions [{
      collisionType=CustomSphere
      sphereCollisionRadius=1
      boundBone {boneName=dyn_hair_fls_04_end}
      boxExtents=<10 20 20>
      constraintSetup {angularLimitsMax=<0 20 20>, angularLimitsMin=<0 -20 -80>, angularTarget=<0 0 0>}
      localJointOffset=<1 0 0>
     }]
    sphericalLimits [
     {limitRadius=7, drivingBone {boneName=neck_02}, sphereLocalOffset=<-10 0 0>}
     {limitRadius=10, drivingBone {boneName=clavicle_l}, sphereLocalOffset=<10 0 -3>}
     {limitRadius=10, drivingBone {boneName=clavicle_r}, sphereLocalOffset=<-10 0 3>}
    ]
   }
   n37 = Anim Dynamics - Bone: dyn_hair_bls_01_end(ComponentPose=@n36, Alpha=1, ExternalForce="0,0,0") {
    angularDampingOverride=1
    angularSpringConstant=1500
    bAngularSpring=true
    bGravityOverrideInSimSpace=true
    bOverrideAngularDamping=true
    bUseSphericalLimits=true
    gravityScale=0.35
    simulationSpace=RootRelative
    alphaBoolBlend {blendOption=Linear}
    alphaScaleBiasClamp {scale=0.5}
    boundBone {boneName=dyn_hair_bls_01_end}
    componentLinearAccScale {z=0.5}
    componentLinearVelScale {z=0.5}
    physicsBodyDefinitions [{
      collisionType=CustomSphere
      sphereCollisionRadius=2
      boundBone {boneName=dyn_hair_bls_01_end}
      boxExtents=<10 20 20>
      constraintSetup {angularLimitsMax=<0 20 20>, angularLimitsMin=<0 -20 -80>, angularTarget=<0 0 0>}
      localJointOffset=<1 0 0>
     }]
    sphericalLimits [
     {limitRadius=7, drivingBone {boneName=neck_02}, sphereLocalOffset=<-10 0 0>}
     {limitRadius=8, drivingBone {boneName=clavicle_r}, sphereLocalOffset=<-10 0 3>}
     {limitRadius=8, drivingBone {boneName=clavicle_l}, sphereLocalOffset=<6.336 -0.5767 -1.5461>}
    ]
   }
   n38 = Anim Dynamics - Bone: dyn_hair_bls_02_end(ComponentPose=@n37, Alpha=1, ExternalForce="0,0,0") {
    angularDampingOverride=1
    angularSpringConstant=1500
    bAngularSpring=true
    bGravityOverrideInSimSpace=true
    bOverrideAngularDamping=true
    bUseSphericalLimits=true
    gravityScale=0.35
    simulationSpace=RootRelative
    alphaBoolBlend {blendOption=Linear}
    alphaScaleBiasClamp {scale=0.5}
    boundBone {boneName=dyn_hair_bls_02_end}
    componentLinearAccScale {z=0.5}
    componentLinearVelScale {z=0.5}
    physicsBodyDefinitions [{
      collisionType=CustomSphere
      sphereCollisionRadius=2
      boundBone {boneName=dyn_hair_bls_02_end}
      boxExtents=<10 20 20>
      constraintSetup {angularLimitsMax=<0 20 20>, angularLimitsMin=<0 -80 -80>, angularTarget=<0 0 0>}
      localJointOffset=<1 0 0>
     }]
    sphericalLimits [
     {limitRadius=8, drivingBone {boneName=neck_02}, sphereLocalOffset=<-10 0 0>}
     {limitRadius=8, drivingBone {boneName=clavicle_l}, sphereLocalOffset=<10 0 -3>}
     {limitRadius=7, drivingBone {boneName=clavicle_r}, sphereLocalOffset=<-10 0 3>}
    ]
   }
   n39 = Anim Dynamics - Bone: dyn_hair_bls_03_end(ComponentPose=@n38, Alpha=1, ExternalForce="0,0,0") {
    angularDampingOverride=1
    angularSpringConstant=1500
    bAngularSpring=true
    bGravityOverrideInSimSpace=true
    bOverrideAngularDamping=true
    bUseSphericalLimits=true
    gravityScale=0.35
    simulationSpace=RootRelative
    alphaBoolBlend {blendOption=Linear}
    alphaScaleBiasClamp {scale=0.5}
    boundBone {boneName=dyn_hair_bls_03_end}
    componentLinearAccScale {z=0.5}
    componentLinearVelScale {z=0.5}
    physicsBodyDefinitions [{
      collisionType=CustomSphere
      sphereCollisionRadius=2
      boundBone {boneName=dyn_hair_bls_03_end}
      boxExtents=<10 20 20>
      constraintSetup {angularLimitsMax=<0 20 20>, angularLimitsMin=<0 -20 -20>, angularTarget=<0 0 0>}
      localJointOffset=<1 0 0>
     }]
    sphericalLimits [
     {limitRadius=8, drivingBone {boneName=neck_02}, sphereLocalOffset=<-10 0 0>}
     {limitRadius=8, drivingBone {boneName=clavicle_l}, sphereLocalOffset=<10 0 -3>}
     {limitRadius=7, drivingBone {boneName=clavicle_r}, sphereLocalOffset=<-10 0 3>}
    ]
   }
   n40 = Anim Dynamics - Bone: dyn_hair_nbk_01(ComponentPose=@n39, Alpha=1, ExternalForce="0,0,0") {
    angularDampingOverride=1
    angularSpringConstant=1500
    bAngularSpring=true
    bGravityOverrideInSimSpace=true
    bOverrideAngularDamping=true
    bUseSphericalLimits=true
    gravityScale=0.35
    simulationSpace=RootRelative
    alphaBoolBlend {blendOption=Linear}
    alphaScaleBiasClamp {scale=0.5}
    boundBone {boneName=dyn_hair_nbk_01}
    componentLinearAccScale {z=0.5}
    componentLinearVelScale {z=0.5}
    physicsBodyDefinitions [{
      collisionType=CustomSphere
      sphereCollisionRadius=2
      boundBone {boneName=dyn_hair_nbk_01}
      boxExtents=<10 20 20>
      constraintSetup {angularLimitsMax=<0 20 20>, angularLimitsMin=<0 -20 -20>, angularTarget=<0 0 0>}
      localJointOffset=<10 0 0>
     }]
    sphericalLimits [
     {limitRadius=10, drivingBone {boneName=clavicle_l}, sphereLocalOffset=<10 0 -3>}
     {limitRadius=10, drivingBone {boneName=clavicle_r}, sphereLocalOffset=<-10 0 3>}
     {limitRadius=12, drivingBone {boneName=spine_05}, sphereLocalOffset=<0 0 0>}
     {limitRadius=7, drivingBone {boneName=neck_02}, sphereLocalOffset=<0 0 0>}
    ]
   }
   n41 = Anim Dynamics - Bone: dyn_hair_nbk_02(ComponentPose=@n40, Alpha=1, ExternalForce="0,0,0") {
    angularDampingOverride=1
    angularSpringConstant=1500
    bAngularSpring=true
    bGravityOverrideInSimSpace=true
    bOverrideAngularDamping=true
    bUseSphericalLimits=true
    gravityScale=0.35
    simulationSpace=RootRelative
    alphaBoolBlend {blendOption=Linear}
    alphaScaleBiasClamp {scale=0.5}
    boundBone {boneName=dyn_hair_nbk_02}
    componentLinearAccScale {z=0.5}
    componentLinearVelScale {z=0.5}
    physicsBodyDefinitions [{
      collisionType=CustomSphere
      sphereCollisionRadius=2
      boundBone {boneName=dyn_hair_nbk_02}
      boxExtents=<10 20 20>
      constraintSetup {angularLimitsMax=<0 20 20>, angularLimitsMin=<0 -20 -20>, angularTarget=<0 0 0>}
      localJointOffset=<0 0 -5>
     }]
    sphericalLimits [
     {limitRadius=10, drivingBone {boneName=clavicle_l}, sphereLocalOffset=<10 0 -3>}
     {limitRadius=10, drivingBone {boneName=clavicle_r}, sphereLocalOffset=<-10 0 3>}
     {limitRadius=5, drivingBone {boneName=neck_02}, sphereLocalOffset=<0 0 0>}
    ]
   }
   n42 = Anim Dynamics - Bone: dyn_hair_bls_02_mid(ComponentPose=@n41, Alpha=1, ExternalForce="0,0,0") {
    angularDampingOverride=1
    angularSpringConstant=1500
    bAngularSpring=true
    bGravityOverrideInSimSpace=true
    bOverrideAngularDamping=true
    bUseSphericalLimits=true
    gravityScale=0.35
    simulationSpace=RootRelative
    alphaBoolBlend {blendOption=Linear}
    alphaScaleBiasClamp {scale=0.5}
    boundBone {boneName=dyn_hair_bls_02_mid}
    componentLinearAccScale {z=0.5}
    componentLinearVelScale {z=0.5}
    physicsBodyDefinitions [{
      collisionType=CustomSphere
      sphereCollisionRadius=2
      boundBone {boneName=dyn_hair_bls_02_mid}
      boxExtents=<10 20 20>
      constraintSetup {angularLimitsMax=<0 20 45>, angularLimitsMin=<0 -80 -45>, angularTarget=<0 0 0>}
      localJointOffset=<2 0 0>
     }]
    sphericalLimits [
     {limitRadius=7, drivingBone {boneName=clavicle_l}, sphereLocalOffset=<10 0 -3>}
     {limitRadius=10, drivingBone {boneName=clavicle_r}, sphereLocalOffset=<-10 0 3>}
     {limitRadius=6, drivingBone {boneName=neck_02}, sphereLocalOffset=<0 0 0>}
     {limitRadius=14, drivingBone {boneName=spine_05}, sphereLocalOffset=<0 0 0>}
    ]
   }
   n43 = Anim Dynamics - Bone: dyn_hair_r_03_mid(ComponentPose=@n42, Alpha=1, ExternalForce="0,0,0") {
    angularDampingOverride=1
    angularSpringConstant=1500
    bAngularSpring=true
    bGravityOverrideInSimSpace=true
    bOverrideAngularDamping=true
    bUseSphericalLimits=true
    gravityScale=0.35
    simulationSpace=RootRelative
    alphaBoolBlend {blendOption=Linear}
    alphaScaleBiasClamp {scale=0.5}
    boundBone {boneName=dyn_hair_r_03_mid}
    componentLinearAccScale {z=0.5}
    componentLinearVelScale {z=0.5}
    physicsBodyDefinitions [{
      collisionType=CustomSphere
      sphereCollisionRadius=2
      boundBone {boneName=dyn_hair_r_03_mid}
      boxExtents=<10 20 20>
      constraintSetup {angularLimitsMax=<0 20 0>, angularLimitsMin=<0 -80 -80>, angularTarget=<0 0 0>}
      localJointOffset=<2 0 0>
     }]
    sphericalLimits [
     {limitRadius=10, drivingBone {boneName=neck_02}, sphereLocalOffset=<-10 0 0>}
     {limitRadius=10, drivingBone {boneName=clavicle_l}, sphereLocalOffset=<5 0 -3>}
     {limitRadius=10, drivingBone {boneName=clavicle_r}, sphereLocalOffset=<-10 5.1379 6.6797>}
    ]
   }
   n44 = Anim Dynamics - Bone: dyn_hair_bk_01_mid(ComponentPose=@n43, Alpha=1, ExternalForce="0,0,0") {
    angularDampingOverride=1
    angularSpringConstant=1500
    bAngularSpring=true
    bGravityOverrideInSimSpace=true
    bOverrideAngularDamping=true
    bUseSphericalLimits=true
    gravityScale=0.35
    simulationSpace=RootRelative
    alphaBoolBlend {blendOption=Linear}
    alphaScaleBiasClamp {scale=0.5}
    boundBone {boneName=dyn_hair_bk_01_mid}
    componentLinearAccScale {z=0.5}
    componentLinearVelScale {z=0.5}
    physicsBodyDefinitions [{
      collisionType=CustomSphere
      sphereCollisionRadius=2
      boundBone {boneName=dyn_hair_bk_01_mid}
      boxExtents=<10 20 20>
      constraintSetup {angularLimitsMax=<0 40 10>, angularLimitsMin=<0 -40 -80>, angularTarget=<0 0 0>}
      localJointOffset=<5 0 0>
     }]
    sphericalLimits [
     {limitRadius=10, drivingBone {boneName=clavicle_l}, sphereLocalOffset=<10 0 -3>}
     {limitRadius=10, drivingBone {boneName=clavicle_r}, sphereLocalOffset=<-10 0 3>}
     {limitRadius=7, drivingBone {boneName=neck_02}, sphereLocalOffset=<0 0 0>}
     {limitRadius=14, drivingBone {boneName=spine_05}, sphereLocalOffset=<0 0 0>}
    ]
   }
   n45 = Anim Dynamics - Bone: dyn_hair_bk_02_mid(ComponentPose=@n44, Alpha=1, ExternalForce="0,0,0") {
    angularDampingOverride=1
    angularSpringConstant=1500
    bAngularSpring=true
    bGravityOverrideInSimSpace=true
    bOverrideAngularDamping=true
    bUseSphericalLimits=true
    gravityScale=0.35
    simulationSpace=RootRelative
    alphaBoolBlend {blendOption=Linear}
    alphaScaleBiasClamp {scale=0.5}
    boundBone {boneName=dyn_hair_bk_02_mid}
    componentLinearAccScale {z=0.5}
    componentLinearVelScale {z=0.5}
    physicsBodyDefinitions [{
      collisionType=CustomSphere
      sphereCollisionRadius=2
      boundBone {boneName=dyn_hair_bk_02_mid}
      boxExtents=<10 20 20>
      constraintSetup {angularLimitsMax=<0 40 45>, angularLimitsMin=<0 -40 -45>, angularTarget=<0 0 0>}
      localJointOffset=<5 0 0>
     }]
    sphericalLimits [
     {limitRadius=10, drivingBone {boneName=clavicle_l}, sphereLocalOffset=<10 0 -3>}
     {limitRadius=10, drivingBone {boneName=clavicle_r}, sphereLocalOffset=<-10 0 3>}
     {limitRadius=7, drivingBone {boneName=neck_02}, sphereLocalOffset=<0 0 0>}
     {limitRadius=14, drivingBone {boneName=spine_05}, sphereLocalOffset=<0 0 0>}
    ]
   }
   n46 = Get skydive_alpha() -> (skydive_alpha:double)
   n47 = "Constraint - dyn_hair_bk_01 (R)"(ComponentPose=@n45, Alpha=@n46, ConstraintWeights_0=1) {
    alphaBoolBlend {blendOption=Linear}
    boneToModify {boneName=dyn_hair_bk_01}
    constraintSetup [{transformType=Rotation, targetBone {boneName=neck_01}}]
    constraintWeights [1]
   }
   n48 = "Constraint - dyn_hair_bk_02 (R)"(ComponentPose=@n47, Alpha=@n46, ConstraintWeights_0=1) {
    alphaBoolBlend {blendOption=Linear}
    boneToModify {boneName=dyn_hair_bk_02}
    constraintSetup [{transformType=Rotation, targetBone {boneName=neck_01}}]
    constraintWeights [1]
   }
   n49 = "Transform (Modify) Bone - Bone: dyn_hair_bk_01"(ComponentPose=@n48, Alpha=@n46, Translation="0,0,0", Rotation="-0.000,-17.826,0.000", Scale="0.8,0.8,0.8") {
    rotationMode=BMM_Additive
    rotationSpace=BCS_BoneSpace
    scaleMode=BMM_Additive
    scaleSpace=BCS_BoneSpace
    alphaBoolBlend {blendOption=Linear}
    alphaScaleBiasClamp {bInterpResult=true}
    boneToModify {boneName=dyn_hair_bk_01}
    rotation=<rot 0 -17.826 0>
    scale=<0.8 0.8 0.8>
   }
   n50 = "Transform (Modify) Bone - Bone: dyn_hair_bk_02"(ComponentPose=@n49, Alpha=@n46, Translation="0,0,0", Rotation="-0.000,4.787,-0.000", Scale="0.8,0.8,0.8") {
    rotationMode=BMM_Additive
    rotationSpace=BCS_BoneSpace
    scaleMode=BMM_Additive
    scaleSpace=BCS_BoneSpace
    alphaBoolBlend {blendOption=Linear}
    alphaScaleBiasClamp {bInterpResult=true}
    boneToModify {boneName=dyn_hair_bk_02}
    rotation {yaw=4.787}
    scale=<0.8 0.8 0.8>
   }
   n51 = "Constraint - dyn_hair_bk_01_mid (R)"(ComponentPose=@n50, Alpha=@n46, ConstraintWeights_0=1) {
    alphaBoolBlend {blendOption=Linear}
    boneToModify {boneName=dyn_hair_bk_01_mid}
    constraintSetup [{transformType=Rotation, targetBone {boneName=neck_01}}]
    constraintWeights [1]
   }
   n52 = "Constraint - dyn_hair_bk_02_mid (R)"(ComponentPose=@n51, Alpha=@n46, ConstraintWeights_0=1) {
    alphaBoolBlend {blendOption=Linear}
    boneToModify {boneName=dyn_hair_bk_02_mid}
    constraintSetup [{transformType=Rotation, targetBone {boneName=neck_01}}]
    constraintWeights [1]
   }
   n53 = "Constraint - dyn_hair_bk_01_mid (T)"(ComponentPose=@n52, Alpha=@n46, ConstraintWeights_0=1) {
    alphaBoolBlend {blendOption=Linear}
    alphaScaleBias {scale=0.1}
    boneToModify {boneName=dyn_hair_bk_01_mid}
    constraintSetup [{transformType=Translation, targetBone {boneName=spine_05}}]
    constraintWeights [1]
   }
   n54 = "Constraint - dyn_hair_bk_02_mid (T)"(ComponentPose=@n53, Alpha=@n46, ConstraintWeights_0=1) {
    alphaBoolBlend {blendOption=Linear}
    alphaScaleBias {scale=0.1}
    boneToModify {boneName=dyn_hair_bk_02_mid}
    constraintSetup [{transformType=Translation, targetBone {boneName=spine_05}}]
    constraintWeights [1]
   }
   n55 = "Constraint - dyn_hair_bls_cst (RR)"!(ComponentPose=@n54, Alpha=@n46, ConstraintWeights_0=0, ConstraintWeights_1:float=1) -> (Pose) {
    alphaBoolBlend {blendOption=Linear}
    boneToModify {boneName=dyn_hair_bls_cst}
    constraintSetup [{transformType=Rotation, targetBone {boneName=spine_05}}, {transformType=Rotation, targetBone {boneName=clavicle_l}}]
    constraintWeights [1,1]
   }
   n56 = "Transform (Modify) Bone - Bone: dyn_hair_bls_cst"(ComponentPose=@n55, Alpha=@n46, Translation="0,0,0", Rotation="12.732,25.009,5.966", Scale="0.7,0.7,0.7") {
    rotationMode=BMM_Additive
    rotationSpace=BCS_BoneSpace
    scaleMode=BMM_Additive
    scaleSpace=BCS_BoneSpace
    alphaBoolBlend {blendOption=Linear}
    alphaScaleBiasClamp {bInterpResult=true}
    boneToModify {boneName=dyn_hair_bls_cst}
    rotation=<rot 12.732 25.009 5.966>
    scale=<0.7 0.7 0.7>
   }
   n57 = "Constraint - dyn_hair_r_03 (RR)"!(ComponentPose=@n56, Alpha=@n46, ConstraintWeights_0=0, ConstraintWeights_1:float=1) -> (Pose) {
    alphaBoolBlend {blendOption=Linear}
    boneToModify {boneName=dyn_hair_r_03}
    constraintSetup [{transformType=Rotation, targetBone {boneName=spine_05}}, {transformType=Rotation, targetBone {boneName=clavicle_r}}]
    constraintWeights [1,1]
   }
   n58 = "Transform (Modify) Bone - Bone: dyn_hair_r_03"(ComponentPose=@n57, Alpha=@n46, Translation="0,0,0", Rotation="-6.682,-37.249,-6.129", Scale="0.7,0.7,0.7") {
    rotationMode=BMM_Additive
    rotationSpace=BCS_BoneSpace
    scaleMode=BMM_Additive
    scaleSpace=BCS_BoneSpace
    alphaBoolBlend {blendOption=Linear}
    alphaScaleBiasClamp {bInterpResult=true}
    boneToModify {boneName=dyn_hair_r_03}
    rotation=<rot -6.682 -37.249 -6.129>
    scale=<0.7 0.7 0.7>
   }
   n59 = "Constraint - dyn_hair_frs (R)"(ComponentPose=@n58, Alpha=@n46, ConstraintWeights_0=1) {
    alphaBoolBlend {blendOption=Linear}
    alphaScaleBias {scale=0.5}
    boneToModify {boneName=dyn_hair_frs}
    constraintSetup [{transformType=Rotation, targetBone {boneName=dyn_hair_r_03}}]
    constraintWeights [1]
   }
   n60 = "Constraint - dyn_hair_fls (R)"(ComponentPose=@n59, Alpha=@n46, ConstraintWeights_0=1) {
    alphaBoolBlend {blendOption=Linear}
    alphaScaleBias {scale=0.5}
    boneToModify {boneName=dyn_hair_fls}
    constraintSetup [{transformType=Rotation, targetBone {boneName=dyn_hair_bls_cst}}]
    constraintWeights [1]
   }
   n61 = Get bIsParachuteOpened()
   n62 = "Transform (Modify) Bone - Bone: dyn_hair_bls"!(ComponentPose=@n60, bAlphaBoolEnabled:bool=@n61, Translation="0,0,0", Rotation="-61.265,-32.385,27.313", Scale="1,1,1") -> (Pose) {
    alphaInputType=Bool
    rotationMode=BMM_Additive
    rotationSpace=BCS_BoneSpace
    alphaBoolBlend {blendInTime=0.25, blendOption=Linear, blendOutTime=0.25}
    boneToModify {boneName=dyn_hair_bls}
    rotation=<rot -61.2651 -32.3849 27.3129>
   }
   n63 = "Transform (Modify) Bone - Bone: dyn_hair_fls_cst"!(ComponentPose=@n62, bAlphaBoolEnabled:bool=@n61, Translation="0,0,0", Rotation="-15.000,0.000,-0.001", Scale="1,1,1") -> (Pose) {
    alphaInputType=Bool
    rotationMode=BMM_Additive
    rotationSpace=BCS_BoneSpace
    alphaBoolBlend {blendInTime=0.25, blendOption=Linear, blendOutTime=0.25}
    boneToModify {boneName=dyn_hair_fls_cst}
    rotation=<rot -15 0 -0.001>
   }
   n64 = "Transform (Modify) Bone - Bone: dyn_hair_frs"!(ComponentPose=@n63, bAlphaBoolEnabled:bool=@n61, Translation="0,0,0", Rotation="-20.000,-0.000,-0.001", Scale="1,1,1") -> (Pose) {
    alphaInputType=Bool
    rotationMode=BMM_Additive
    rotationSpace=BCS_BoneSpace
    alphaBoolBlend {blendInTime=0.25, blendOption=Linear, blendOutTime=0.25}
    boneToModify {boneName=dyn_hair_frs}
    rotation=<rot -19.9999 -0.0002 -0.0008>
   }
   n65 = "Transform (Modify) Bone - Bone: hair_r_03_zero"!(ComponentPose=@n64, bAlphaBoolEnabled:bool=@n61, Translation="0,0,0", Rotation="-13.990,51.933,-0.681", Scale="1,1,1") -> (Pose) {
    alphaInputType=Bool
    rotationMode=BMM_Additive
    rotationSpace=BCS_BoneSpace
    alphaBoolBlend {blendInTime=0.25, blendOption=Linear, blendOutTime=0.25}
    boneToModify {boneName=hair_r_03_zero}
    rotation=<rot -13.99 51.9329 -0.681>
   }
   Output Pose(Result=Component To Local(ComponentPose=@n65)) {layerGroup=None}
   n66 = "Transform (Modify) Bone - Bone: upperarm_l"(
    ComponentPose=Local To Component(LocalPose=Copy Pose From Mesh() {bUseAttachedParent=true})
    Alpha=1
    Translation=Get upperarm l translation() -> (upperarm l translation:Vector)
    Rotation=Get upperarm l rotation() -> (upperarm l rotation:Rotator)
    Scale="1,1,1"
   ) {
    rotationMode=BMM_Replace
    rotationSpace=BCS_WorldSpace
    translationMode=BMM_Replace
    translationSpace=BCS_WorldSpace
    alphaBoolBlend {blendOption=Linear}
    boneToModify {boneName=upperarm_l}
   }
   n67 = "Transform (Modify) Bone - Bone: upperarm_r"(
    ComponentPose=@n66
    Alpha=1
    Translation=Get upperarm r translation() -> (upperarm r translation:Vector)
    Rotation=Get upperarm r rotation() -> (upperarm r rotation:Rotator)
    Scale="1,1,1"
   ) {
    rotationMode=BMM_Replace
    rotationSpace=BCS_WorldSpace
    translationMode=BMM_Replace
    translationSpace=BCS_WorldSpace
    alphaBoolBlend {blendOption=Linear}
    boneToModify {boneName=upperarm_r}
   }
   n68 = "Constraint - hair_bk_01_tgt (PP)"!(ComponentPose=@n67, Alpha=1, ConstraintWeights_0=0, ConstraintWeights_1:float=1) -> (Pose) {
    alphaBoolBlend {blendOption=Linear}
    boneToModify {boneName=hair_bk_01_tgt}
    constraintSetup [{targetBone {boneName=clavicle_l}}, {targetBone {boneName=spine_05}}]
    constraintWeights [1,1]
   }
   n69 = "Constraint - hair_bk_02_tgt (PP)"!(ComponentPose=@n68, Alpha=1, ConstraintWeights_0=0, ConstraintWeights_1:float=1) -> (Pose) {
    alphaBoolBlend {blendOption=Linear}
    boneToModify {boneName=hair_bk_02_tgt}
    constraintSetup [{targetBone {boneName=clavicle_r}}, {targetBone {boneName=spine_05}}]
    constraintWeights [1,1]
   }
   n70 = "Constraint - hair_r_03_tgt (PP)"!(ComponentPose=@n69, Alpha=1, ConstraintWeights_0=0.6, ConstraintWeights_1:float=0.4) -> (Pose) {
    alphaBoolBlend {blendOption=Linear}
    boneToModify {boneName=hair_r_03_tgt}
    constraintSetup [{targetBone {boneName=clavicle_r}}, {targetBone {boneName=spine_05}}]
    constraintWeights [1,1]
   }
   n71 = "Constraint - hair_bls_tgt (PP)"!(ComponentPose=@n70, Alpha=1, ConstraintWeights_0=0.6, ConstraintWeights_1:float=0.4) -> (Pose) {
    alphaBoolBlend {blendOption=Linear}
    boneToModify {boneName=hair_bls_tgt}
    constraintSetup [{targetBone {boneName=clavicle_l}}, {targetBone {boneName=spine_05}}]
    constraintWeights [1,1]
   }
   n72 = "Constraint - hair_fls_tgt (PP)"!(ComponentPose=@n71, Alpha=1, ConstraintWeights_0=0.7, ConstraintWeights_1:float=0.3) -> (Pose) {
    alphaBoolBlend {blendOption=Linear}
    boneToModify {boneName=hair_fls_tgt}
    constraintSetup [{targetBone {boneName=clavicle_l}}, {targetBone {boneName=spine_05}}]
    constraintWeights [1,1]
   }
   n73 = "Constraint - hair_frs_tgt (PP)"!(ComponentPose=@n72, Alpha=1, ConstraintWeights_0=0.7, ConstraintWeights_1:float=0.3) -> (Pose) {
    alphaBoolBlend {blendOption=Linear}
    boneToModify {boneName=hair_frs_tgt}
    constraintSetup [{targetBone {boneName=clavicle_r}}, {targetBone {boneName=spine_05}}]
    constraintWeights [1,1]
   }
   n74 = "Transform (Modify) Bone - Bone: hair_bk_01_tgt"(ComponentPose=@n73, Alpha=1, Translation="10,0,0", Rotation="0,0,0", Scale="1,1,1") {
    translationMode=BMM_Additive
    translationSpace=BCS_BoneSpace
    alphaBoolBlend {blendOption=Linear}
    boneToModify {boneName=hair_bk_01_tgt}
    translation {x=10}
   }
   n75 = "Transform (Modify) Bone - Bone: hair_bk_02_tgt"(ComponentPose=@n74, Alpha=1, Translation="10,0,0", Rotation="0,0,0", Scale="1,1,1") {
    translationMode=BMM_Additive
    translationSpace=BCS_BoneSpace
    alphaBoolBlend {blendOption=Linear}
    boneToModify {boneName=hair_bk_02_tgt}
    translation {x=10}
   }
   n76 = "Transform (Modify) Bone - Bone: hair_r_03_tgt"(ComponentPose=@n75, Alpha=1, Translation="10,0,0", Rotation="0,0,0", Scale="1,1,1") {
    translationMode=BMM_Additive
    translationSpace=BCS_BoneSpace
    alphaBoolBlend {blendOption=Linear}
    boneToModify {boneName=hair_r_03_tgt}
    translation {x=10}
   }
   n77 = "Transform (Modify) Bone - Bone: hair_bls_tgt"(ComponentPose=@n76, Alpha=1, Translation="10,0,0", Rotation="0,0,0", Scale="1,1,1") {
    translationMode=BMM_Additive
    translationSpace=BCS_BoneSpace
    alphaBoolBlend {blendOption=Linear}
    boneToModify {boneName=hair_bls_tgt}
    translation {x=10}
   }
   n78 = "Transform (Modify) Bone - Bone: hair_fls_tgt"(ComponentPose=@n77, Alpha=1, Translation="10,10,0", Rotation="0,0,0", Scale="1,1,1") {
    translationMode=BMM_Additive
    translationSpace=BCS_BoneSpace
    alphaBoolBlend {blendOption=Linear}
    boneToModify {boneName=hair_fls_tgt}
    translation=<10 10>
   }
   n79 = "Transform (Modify) Bone - Bone: hair_frs_tgt"(ComponentPose=@n78, Alpha=1, Translation="10,0,0", Rotation="0,0,0", Scale="1,1,1") {
    translationMode=BMM_Additive
    translationSpace=BCS_BoneSpace
    alphaBoolBlend {blendOption=Linear}
    boneToModify {boneName=hair_frs_tgt}
    translation {x=10}
   }
   n80 = "Transform (Modify) Bone - Bone: hair_r_03_tgt"!(ComponentPose=@n79, bAlphaBoolEnabled:bool=Get bIsTargeting(), Translation="0,0,20", Rotation="0,0,0", Scale="1,1,1") -> (Pose) {
    alphaInputType=Bool
    translationMode=BMM_Additive
    translationSpace=BCS_BoneSpace
    alphaBoolBlend {blendOption=Linear, blendOutTime=1}
    boneToModify {boneName=hair_r_03_tgt}
    translation {z=20}
   }
   n81 = Get skydive_alpha()
   n82 = "Transform (Modify) Bone - Bone: hair_bk_01_tgt"(ComponentPose=@n80, Alpha=@n81, Translation="0,0,10", Rotation="0,0,0", Scale="1,1,1") {
    translationMode=BMM_Additive
    translationSpace=BCS_BoneSpace
    alphaBoolBlend {blendOption=Linear}
    alphaScaleBiasClamp {bInterpResult=true}
    boneToModify {boneName=hair_bk_01_tgt}
    translation {z=10}
   }
   n83 = "Transform (Modify) Bone - Bone: hair_bk_02_tgt"(ComponentPose=@n82, Alpha=@n81, Translation="0,0,10", Rotation="0,0,0", Scale="1,1,1") {
    translationMode=BMM_Additive
    translationSpace=BCS_BoneSpace
    alphaBoolBlend {blendOption=Linear}
    alphaScaleBiasClamp {bInterpResult=true}
    boneToModify {boneName=hair_bk_02_tgt}
    translation {z=10}
   }
   n84 = "Transform (Modify) Bone - Bone: hair_r_03_tgt"!(ComponentPose=@n83, Alpha:float=@n81, Translation="-10,0,20", Rotation="0,0,0", Scale="1,1,1") -> (Pose) {
    translationMode=BMM_Additive
    translationSpace=BCS_BoneSpace
    alphaBoolBlend {blendOption=Linear}
    alphaScaleBiasClamp {bInterpResult=true}
    boneToModify {boneName=hair_r_03_tgt}
    translation {x=-10, z=20}
   }
   n85 = "Transform (Modify) Bone - Bone: hair_bls_tgt"(ComponentPose=@n84, Alpha=@n81, Translation="-10,0,20", Rotation="0,0,0", Scale="1,1,1") {
    translationMode=BMM_Additive
    translationSpace=BCS_BoneSpace
    alphaBoolBlend {blendOption=Linear}
    alphaScaleBiasClamp {bInterpResult=true}
    boneToModify {boneName=hair_bls_tgt}
    translation {x=-10, z=20}
   }
   n86 = "Transform (Modify) Bone - Bone: hair_fls_tgt"(ComponentPose=@n85, Alpha=@n81, Translation="-10,0,20", Rotation="0,0,0", Scale="1,1,1") {
    translationMode=BMM_Additive
    translationSpace=BCS_BoneSpace
    alphaBoolBlend {blendOption=Linear}
    alphaScaleBiasClamp {bInterpResult=true}
    boneToModify {boneName=hair_fls_tgt}
    translation {x=-10, z=20}
   }
   n87 = "Transform (Modify) Bone - Bone: hair_frs_tgt"(ComponentPose=@n86, Alpha=@n81, Translation="-10,-10,20", Rotation="0,0,0", Scale="1,1,1") {
    translationMode=BMM_Additive
    translationSpace=BCS_BoneSpace
    alphaBoolBlend {blendOption=Linear}
    alphaScaleBiasClamp {bInterpResult=true}
    boneToModify {boneName=hair_frs_tgt}
    translation=<-10 -10 20>
   }
   n88 = Get skydiving_backpack() -> (skydiving_backpack:double)
   n89 = "Transform (Modify) Bone - Bone: hair_bk_01_tgt"(ComponentPose=@n87, Alpha=@n88, Translation="-10,0,40", Rotation="0,0,0", Scale="1,1,1") {
    translationMode=BMM_Additive
    translationSpace=BCS_BoneSpace
    alphaBoolBlend {blendOption=Linear}
    alphaScaleBiasClamp {bInterpResult=true}
    boneToModify {boneName=hair_bk_01_tgt}
    translation {x=-10, z=40}
   }
   n90 = "Transform (Modify) Bone - Bone: hair_bk_02_tgt"(ComponentPose=@n89, Alpha=@n88, Translation="-10,0,40", Rotation="0,0,0", Scale="1,1,1") {
    translationMode=BMM_Additive
    translationSpace=BCS_BoneSpace
    alphaBoolBlend {blendOption=Linear}
    alphaScaleBiasClamp {bInterpResult=true}
    boneToModify {boneName=hair_bk_02_tgt}
    translation {x=-10, z=40}
   }
   n91 = "Transform (Modify) Bone - Bone: hair_r_03_tgt"(ComponentPose=@n90, Alpha=1, Translation="0.767,22.542,6.245", Rotation="0,0,0", Scale="1,1,1") {
    translationMode=BMM_Additive
    translationSpace=BCS_BoneSpace
    alphaBoolBlend {blendOption=Linear}
    boneToModify {boneName=hair_r_03_tgt}
    translation=<0.767 22.542 6.245>
   }
   n92 = "Transform (Modify) Bone - Bone: hair_frs_tgt"(ComponentPose=@n91, Alpha=1, Translation="-22.391,-38.467,-31.982", Rotation="0,0,0", Scale="1,1,1") {
    translationMode=BMM_Additive
    translationSpace=BCS_BoneSpace
    alphaBoolBlend {blendOption=Linear}
    boneToModify {boneName=hair_frs_tgt}
    translation=<-22.391 -38.467 -31.982>
   }
   n93 = Look At - Bone: dyn_hair_bk_01(ComponentPose=@n92, Alpha=1) {
    alphaBoolBlend {blendOption=Linear}
    boneToModify {boneName=dyn_hair_bk_01}
    lookAtLocation {x=0}
    lookAtTarget {boneReference {boneName=hair_bk_01_tgt}}
    lookAt_Axis {axis=<1 0>}
   }
   n94 = Look At - Bone: dyn_hair_bk_02(ComponentPose=@n93, Alpha=1) {
    alphaBoolBlend {blendOption=Linear}
    boneToModify {boneName=dyn_hair_bk_02}
    lookAtLocation {x=0}
    lookAtTarget {boneReference {boneName=hair_bk_02_tgt}}
    lookAt_Axis {axis=<1 0>}
   }
   n95 = Look At - Bone: dyn_hair_r_03(ComponentPose=@n94, Alpha=1) {
    alphaBoolBlend {blendOption=Linear}
    boneToModify {boneName=dyn_hair_r_03}
    lookAtLocation {x=0}
    lookAtTarget {boneReference {boneName=hair_r_03_tgt}}
    lookAt_Axis {axis=<1 0>}
   }
   n96 = Look At - Bone: dyn_hair_bls(ComponentPose=@n95, Alpha=1) {
    alphaBoolBlend {blendOption=Linear}
    boneToModify {boneName=dyn_hair_bls}
    lookAtLocation {x=0}
    lookAtTarget {boneReference {boneName=hair_bls_tgt}}
    lookAt_Axis {axis=<1 0>}
   }
   n97 = Look At - Bone: dyn_hair_fls(ComponentPose=@n96, Alpha=1) {
    alphaBoolBlend {blendOption=Linear}
    boneToModify {boneName=dyn_hair_fls}
    lookAtLocation {x=0}
    lookAtTarget {boneReference {boneName=hair_fls_tgt}}
    lookAt_Axis {axis=<1 0>}
   }
   n98 = Look At - Bone: dyn_hair_frs(ComponentPose=@n97, Alpha=1) {
    alphaBoolBlend {blendOption=Linear}
    boneToModify {boneName=dyn_hair_frs}
    lookAtLocation {x=0}
    lookAtTarget {boneReference {boneName=hair_frs_tgt}}
    lookAt_Axis {axis=<1 0>}
   }
   n99 = Get bIsSkydiveDiveMode()
   n100 = "Transform (Modify) Bone - Bone: dyn_hair_bk_01"!(ComponentPose=@n98, bAlphaBoolEnabled:bool=@n99, Translation="0,0,0", Rotation=Get Rotation() -> (Rotation:Rotator), Scale="1,1,1") -> (Pose) {
    alphaInputType=Bool
    rotationMode=BMM_Additive
    rotationSpace=BCS_BoneSpace
    alphaBoolBlend {blendOption=Linear, blendOutTime=1}
    boneToModify {boneName=dyn_hair_bk_01}
    rotation {pitch=-80}
   }
   n101 = "Transform (Modify) Bone - Bone: dyn_hair_bk_02"!(ComponentPose=@n100, bAlphaBoolEnabled:bool=@n99, Translation="0,0,0", Rotation=Get Rotation(), Scale="1,1,1") -> (Pose) {
    alphaInputType=Bool
    rotationMode=BMM_Additive
    rotationSpace=BCS_BoneSpace
    alphaBoolBlend {blendOption=Linear, blendOutTime=1}
    boneToModify {boneName=dyn_hair_bk_02}
    rotation {pitch=-80}
   }
   Save cached pose 'AimConstraints_Cache'(Pose=Component To Local(ComponentPose=@n101)) {cacheName=AimConstraints_Cache}
  ]}
 EventGraph {nodes [
   Event BlueprintUpdateAnimation()
   > n1 = Sequence!(execute) -> (then_0, then_1, then_2:Exec)
   n2 = Get Is Wearing Backbling()
   n3 = Get bIsSkydiveFloating()
   > Set skydiving_backpack(skydiving_backpack:double=float * float(A="To Float (Boolean)"(InBool=@n3), B="To Float (Boolean)"(InBool=@n2)) {is=title K2Node_PromotableOperator}) -> (Output_Get:double)
   n4 = Get BodyMesh() -> (BodyMesh:Skeletal Mesh Component Object Reference)
   Is Valid(exec=@n1.then_1, InputObject=@n4)
   n5 = BreakTransform(InTransform=GetSocketTransform(self=@n4, InSocketName=upperarm_l, TransformSpace=RTS_World))
   > Set upperarm l translation(upperarm l translation:Vector=@n5) -> (Output_Get:Vector)
   > Set upperarm l rotation(upperarm l rotation:Rotator=@n5.Rotation) -> (Output_Get:Rotator)
   n6 = BreakTransform(InTransform=GetSocketTransform(self=@n4, InSocketName=upperarm_r, TransformSpace=RTS_World))
   > Set upperarm r translation(upperarm r translation:Vector=@n6) -> (Output_Get:Vector)
   > Set upperarm r rotation(upperarm r rotation:Rotator=@n6.Rotation) -> (Output_Get:Rotator)
   n7 = GetOwnerSkeletalMeshComponent()
   n8 = BreakRotator(InRot=BreakTransform(InTransform=GetDeltaTransformFromRefPose(self=@n7, BoneName=upperarm_r, BaseName=clavicle_r)).Rotation)
   n9 = Get bIsSkyDiving()
   Set upperarm r yaw(
    execute=@n1.then_2
    upperarm r yaw:double=SelectFloat(A=MapRangeClamped(Value=@n8.Yaw, InRangeA=50, InRangeB=120, OutRangeB=0.5), B=0, bPickA=NOT Boolean(A=@n9))
   ) -> (Output_Get:double)
   n10 = Event OnCharacterCustomizationLoaded()
   > n11 = Set Is Wearing Backbling(Is Wearing Backbling=CheckForBackbling())
   > Set Rotation(Rotation:Rotator=Select(Option 0:Rotator="-90, 0, 0", Option 1:Rotator="-70, 0, 0", Index=@n11) -> (ReturnValue:Rotator)) -> (Output_Get:Rotator)
   > Set BodyMesh(BodyMesh=Get Array Item with EFort Custom Part Type(Array=@n10.CustomMeshParts))
   Event OnEndSkydiving()
   > Set skydive_alpha(skydive_alpha:double=0) -> (Output_Get:double)
   Event OnBeginSkydiving()
   > Set skydive_alpha(skydive_alpha=1)
  ]}
]
variables [
 bIsTargeting_0 {type=boolean}
 BodyMesh {type=SkeletalMeshComponent}
 bUsesDayPhaseChange_0 {type=boolean}
 bUsesDayPhaseChange_1 {type=boolean}
 bUsesDayPhaseChange_2 {type=boolean}
 Is Wearing Backbling {type=boolean}
 Rotation {type=Rotator, default=<rot -80 0 0>}
 skydive_alpha {type=number}
 skydiving_backpack {type=number}
 upperarm l rotation {type=Rotator}
 upperarm l translation {type=Vector}
 upperarm r rotation {type=Rotator}
 upperarm r translation {type=Vector}
 upperarm r yaw {type=number}
]
